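# Sound cue for collision impacts on the Ferret vehicle in Athena.
# A Random node picks among four equally weighted Ferret collision wave variants.
# It plays in the Fort_Vehicles sound class with a slight volume boost and reuses the golf cart close-movement attenuation asset for distance falloff.
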SoundCue /Game/Athena/Sounds/Vehicles/Ferret/Ferret_Collision_Impact_Cue
root=Random_0
nodes [
 {id=Random_0, children [WavePlayer_1,WavePlayer_2,WavePlayer_3,WavePlayer_4], properties {weights [1,1,1,1]}}
 {id=WavePlayer_1, properties {soundWaveAssetPtr=&/Ferret_Collision_01}}
 {id=WavePlayer_2, properties {soundWaveAssetPtr=&/Ferret_Collision_02}}
 {id=WavePlayer_3, properties {soundWaveAssetPtr=&/Ferret_Collision_03}}
 {id=WavePlayer_4, properties {soundWaveAssetPtr=&/Ferret_Collision_04}}
]
properties {
 attenuationSettings=/Game/Sounds/Attenuation/Fort_GolfCart_Movement_Close_Att
 maxDistance=5300
 soundClassObject=&Fort_Vehicles
 volumeMultiplier=1.2
}
