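# Sound cue for the Ferret vehicle destroying a building, using a borrowed Carmine punch-blast sound.
# A quality-level root plays a mix on its first two levels: a random one of four Ferret collisions plus the pitch- and volume-modulated blast.
# The last two levels play only the optimized blast variant.
# It uses golf-cart close attenuation and a concurrency override that caps instances at two, stopping the oldest.
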SoundCue /Game/Athena/Sounds/Vehicles/Ferret/Ferret_DestroyBuilding_Cue
root=QualityLevel_0
nodes [
 {id=Mixer_0, children [Random_0,Modulator_0], properties {inputVolume [0.5,1]}}
 {id=Modulator_0, children [WavePlayer_2], properties {pitchMax=1.1, pitchMin=0.85, volumeMin=0.7}}
 {id=Modulator_2, children [WavePlayer_0], properties {pitchMax=1, pitchMin=1}}
 {id=QualityLevel_0, children [Mixer_0,Mixer_0,Modulator_2,Modulator_2]}
 {id=Random_0, children [WavePlayer_1,WavePlayer_3,WavePlayer_4,WavePlayer_5], properties {weights [1,1,1,1]}}
 {id=WavePlayer_0, properties {soundWaveAssetPtr=/Game/Sounds/Fort_GamePlay_Sounds/Carmine/Carmine_Punch_Blast_02_opt}}
 {id=WavePlayer_1, properties {soundWaveAssetPtr=&/Ferret_Collision_01}}
 {id=WavePlayer_2, properties {soundWaveAssetPtr=/Game/Sounds/Fort_GamePlay_Sounds/Carmine/Carmine_Punch_Blast_02}}
 {id=WavePlayer_3, properties {soundWaveAssetPtr=&/Ferret_Collision_02}}
 {id=WavePlayer_4, properties {soundWaveAssetPtr=&/Ferret_Collision_03}}
 {id=WavePlayer_5, properties {soundWaveAssetPtr=&/Ferret_Collision_04}}
]
properties {
 attenuationSettings=/Game/Sounds/Attenuation/Fort_GolfCart_Movement_Close_Att
 bOverrideConcurrency=true
 maxDistance=5300
 soundClassObject=&Fort_Mission_Ingame_Important_PSM
 volumeMultiplier=0.4
 concurrencyOverrides {maxCount=2, resolutionRule=StopOldest, volumeScale=0.5}
}
