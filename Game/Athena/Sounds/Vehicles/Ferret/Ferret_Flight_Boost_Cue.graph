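# Sound cue for the flight boost of the Ferret vehicle in Athena.
# A wave player plays the single Ferret_Boost_01 sample through a modulator that randomizes pitch slightly around normal.
# The cue uses the vehicles sound class at reduced volume and borrows a close biplane movement attenuation for spatial falloff.
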SoundCue /Game/Athena/Sounds/Vehicles/Ferret/Ferret_Flight_Boost_Cue
root=Modulator_0
nodes [
 {id=Modulator_0, children [WavePlayer_1], properties {pitchMax=1.1, pitchMin=0.9}}
 {id=WavePlayer_1, properties {soundWaveAssetPtr=&/Ferret_Boost_01}}
]
properties {
 attenuationSettings=&/Fort_Biplane_Movement_MLTI_Close_Att
 maxDistance=18000
 soundClassObject=&Fort_Vehicles
 volumeMultiplier=0.5
}
