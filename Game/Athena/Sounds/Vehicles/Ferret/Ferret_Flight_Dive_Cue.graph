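# Looping dive sound cue for the Ferret vehicle in flight, playing in the vehicles sound class with biplane close-movement attenuation.
# A looping dive wave feeds a continuous modulator that drives volume from the Speed parameter, then a modulator with a fixed raised pitch.
# The quality-level root routes that chain to its first two inputs and leaves the other two empty.
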SoundCue /Game/Athena/Sounds/Vehicles/Ferret/Ferret_Flight_Dive_Cue
root=QualityLevel_0
nodes [
 {id=ModulatorContinuous_0, children [WavePlayer_1], properties {volumeModulationParams {parameterName=Speed}}}
 {id=Modulator_0, children [ModulatorContinuous_0], properties {pitchMax=1.25, pitchMin=1.25}}
 {id=QualityLevel_0, children [Modulator_0,Modulator_0,null,null]}
 {id=WavePlayer_1, properties {bLooping=true, soundWaveAssetPtr=&/Ferret_Dive_01}}
]
properties {
 attenuationSettings=&/Fort_Biplane_Movement_MLTI_Close_Att
 maxDistance=18000
 soundClassObject=&Fort_Vehicles
 volumeMultiplier=0.4
}
