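# Sound cue for the Ferret vehicle's first-person normal flight loop, using biplane interior attenuation and the vehicles sound class.
# A Speed parameter crossfades four looping waves from low through extra-high.
# The blend passes through smoothed Doppler, then two chained continuous modulators that map Boost and Dive parameters to pitch.
# Its concurrency override allows two instances and stops the quietest.
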SoundCue /Game/Athena/Sounds/Vehicles/Ferret/Ferret_Flight_Normal_1p_Cue
root=ModulatorContinuous_2
nodes [
 {id=Doppler_0, children [ParamCrossFade_0], properties {bUseSmoothing=true}}
 {
  id=ModulatorContinuous_1
  children [Doppler_0]
  properties {pitchModulationParams {default=0, maxOutput=1.25, minOutput=1, parameterName=Boost}}
 }
 {
  id=ModulatorContinuous_2
  children [ModulatorContinuous_1]
  properties {pitchModulationParams {default=0, maxOutput=1.25, minOutput=1, parameterName=Dive}}
 }
 {
  id=ParamCrossFade_0
  children [WavePlayer_2,WavePlayer_0,WavePlayer_1,WavePlayer_3]
  properties {
   paramName=Speed
   crossFadeInput [
    {fadeInDistanceEnd=0.1, fadeInDistanceStart=0.05, fadeOutDistanceEnd=0.15, fadeOutDistanceStart=0.1, volume=1}
    {fadeInDistanceEnd=0.2, fadeInDistanceStart=0.1, fadeOutDistanceEnd=0.4, fadeOutDistanceStart=0.25, volume=1}
    {fadeInDistanceEnd=0.4, fadeInDistanceStart=0.25, fadeOutDistanceEnd=0.75, fadeOutDistanceStart=0.6, volume=1}
    {fadeInDistanceEnd=0.75, fadeInDistanceStart=0.6, fadeOutDistanceEnd=100, fadeOutDistanceStart=100, volume=1}
   ]
  }
 }
 {id=WavePlayer_0, properties {bLooping=true, soundWaveAssetPtr=&/Ferret_Med_01}}
 {id=WavePlayer_1, properties {bLooping=true, soundWaveAssetPtr=&/Ferret_High_01}}
 {id=WavePlayer_2, properties {bLooping=true, soundWaveAssetPtr=&/Ferret_Low_01}}
 {id=WavePlayer_3, properties {bLooping=true, soundWaveAssetPtr=&/Ferret_ExtHigh_01}}
]
properties {
 attenuationSettings=&/Fort_Biplane_Movement_MLTI_Interior_Att
 bOverrideConcurrency=true
 maxDistance=9900
 soundClassObject=&Fort_Vehicles
 volumeMultiplier=0.25
 concurrencyOverrides {maxCount=2, resolutionRule=StopQuietest}
}
