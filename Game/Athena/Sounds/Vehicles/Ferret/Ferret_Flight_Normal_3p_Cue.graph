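# Third-person sound cue for normal flight of the Ferret plane, layering close, distant and flyby sounds under separate attenuations.
# The Speed parameter fades in each layer, with looped random flybys arriving at higher speeds.
# Boost raises the close loop's pitch, while Dive raises the close-plus-distant mix, which also gets Doppler.
# The lower two quality levels drop flyby and distant layers, and concurrency stops the quietest beyond three instances.
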SoundCue /Game/Athena/Sounds/Vehicles/Ferret/Ferret_Flight_Normal_3p_Cue
root=Mixer_1
nodes [
 {id=Attenuation_0, children [Looping_0], properties {attenuationSettings=&/Fort_Biplane_Movement_Flyby_Att}}
 {
  id=Attenuation_1
  children [ParamCrossFade_1]
  properties {attenuationSettings=&/Fort_Biplane_Movement_MLTI_Distant_Att}
 }
 {
  id=Attenuation_2
  children [ParamCrossFade_2]
  properties {attenuationSettings=&/Fort_Biplane_Movement_MLTI_Close_Att}
 }
 {id=Doppler_0, children [Mixer_0], properties {bUseSmoothing=true, dopplerIntensity=2}}
 {id=Looping_0, children [QualityLevel_0]}
 {id=Mixer_0, children [ModulatorContinuous_1, Attenuation_1], properties {inputVolume [1,1.5]}}
 {id=Mixer_1, children [ModulatorContinuous_2, ParamCrossFade_0], properties {inputVolume [1,1]}}
 {
  id=ModulatorContinuous_1
  children [Attenuation_2]
  properties {pitchModulationParams {default=0, maxOutput=1.25, minOutput=1, parameterName=Boost}}
 }
 {
  id=ModulatorContinuous_2
  children [Doppler_0]
  properties {pitchModulationParams {default=0, maxOutput=1.25, minOutput=1, parameterName=Dive}}
 }
 {
  id=ParamCrossFade_0
  children [Attenuation_0]
  properties {
   paramName=Speed
   crossFadeInput [{fadeInDistanceEnd=0.5, fadeInDistanceStart=0.15, fadeOutDistanceEnd=100, fadeOutDistanceStart=100, volume=1}]
  }
 }
 {
  id=ParamCrossFade_1
  children [QualityLevel_1]
  properties {
   paramName=Speed
   crossFadeInput [{fadeInDistanceEnd=0.15, fadeInDistanceStart=0, fadeOutDistanceEnd=100, fadeOutDistanceStart=100, volume=1}]
  }
 }
 {
  id=ParamCrossFade_2
  children [null,WavePlayer_7]
  properties {
   paramName=Speed
   crossFadeInput [
    {fadeInDistanceEnd=0, fadeInDistanceStart=0, fadeOutDistanceEnd=0.15, fadeOutDistanceStart=0, volume=1}
    {fadeInDistanceEnd=0.15, fadeInDistanceStart=0, fadeOutDistanceEnd=100, fadeOutDistanceStart=100, volume=1}
   ]
  }
 }
 {id=QualityLevel_0, children [Random_1,Random_1,null,null]}
 {id=QualityLevel_1, children [Random_0,WavePlayer_0,null,null]}
 {id=Random_0, children [WavePlayer_4,WavePlayer_5], properties {weights [1,1]}}
 {id=Random_1, children [WavePlayer_8, WavePlayer_9, WavePlayer_10], properties {weights [1,1,1]}}
 {id=WavePlayer_0, properties {bLooping=true, soundWaveAssetPtr=&/Ferret_Distant_01}}
 {id=WavePlayer_10, properties {soundWaveAssetPtr=&/Ferret_Flyby_03}}
 {id=WavePlayer_4, properties {bLooping=true, soundWaveAssetPtr=&/Ferret_Distant_02}}
 {id=WavePlayer_5, properties {bLooping=true, soundWaveAssetPtr=&Ferret_Distant_01}}
 {id=WavePlayer_7, properties {bLooping=true, soundWaveAssetPtr=&/Ferret_Medium_01}}
 {id=WavePlayer_8, properties {soundWaveAssetPtr=&/Ferret_Flyby_01}}
 {id=WavePlayer_9, properties {soundWaveAssetPtr=&/Ferret_Flyby_02}}
]
properties {
 bOverrideConcurrency=true
 maxDistance=40000
 soundClassObject=&Fort_Vehicles
 concurrencyOverrides {maxCount=3, resolutionRule=StopQuietest}
}
