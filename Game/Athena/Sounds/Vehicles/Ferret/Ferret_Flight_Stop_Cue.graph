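# Sound cue for the Ferret vehicle's flight stop sound, using biplane close-movement attenuation.
# A single wave player is the root and plays Ferret_Start_01, a start-named wave despite the cue's Stop name.
# It routes to the Fort_Vehicles sound class, with no randomization, mixing or modulation nodes.
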SoundCue /Game/Athena/Sounds/Vehicles/Ferret/Ferret_Flight_Stop_Cue
root=WavePlayer_11
nodes [{id=WavePlayer_11, properties {soundWaveAssetPtr=&/Ferret_Start_01}}]
properties {
 attenuationSettings=&/Fort_Biplane_Movement_Close_Att
 maxDistance=5300
 soundClassObject=&Fort_Vehicles
}
