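# Sound cue for the Ferret vehicle's taxi engine loop, played in the vehicles sound class with Biplane movement attenuation.
# It reuses the GolfCart medium engine movement loop as a looping wave.
# A quality-level node at the root plays the standard wave for the first two quality slots and the "_opt" variant for the last two.
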SoundCue /Game/Athena/Sounds/Vehicles/Ferret/Ferret_Flight_Taxi_Cue
root=QualityLevel_0
nodes [
 {id=QualityLevel_0, children [WavePlayer_1,WavePlayer_1,WavePlayer_0,WavePlayer_0]}
 {
  id=WavePlayer_0
  properties {bLooping=true, soundWaveAssetPtr=/Game/Athena/Sounds/Vehicles/GolfCart/GolfCart_Engine_MovementLoop_Medium_01_opt}
 }
 {id=WavePlayer_1, properties {bLooping=true, soundWaveAssetPtr=/Game/Athena/Sounds/Vehicles/GolfCart/GolfCart_Engine_MovementLoop_Medium_01}}
]
properties {
 attenuationSettings=&/Fort_Biplane_Movement_MLTI_Interior_Att
 maxDistance=9900
 soundClassObject=&Fort_Vehicles
 volumeMultiplier=0.76
}
