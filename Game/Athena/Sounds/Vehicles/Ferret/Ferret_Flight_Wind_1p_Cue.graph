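# Sound cue for the first-person flight wind loop of the Ferret vehicle.
# It loops the closed-umbrella falling wind wave through a continuous modulator whose volume follows a Speed parameter.
# It uses the biplane movement interior attenuation and plays in the vehicles sound class.
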SoundCue /Game/Athena/Sounds/Vehicles/Ferret/Ferret_Flight_Wind_1p_Cue
root=ModulatorContinuous_0
nodes [
 {id=ModulatorContinuous_0, children [WavePlayer_4], properties {volumeModulationParams {parameterName=Speed}}}
 {id=WavePlayer_4, properties {bLooping=true, soundWaveAssetPtr=/Game/Athena/Sounds/Wind_Falling/Wind_UmbrellaClosed_Loop_01}}
]
properties {
 attenuationSettings=&/Fort_Biplane_Movement_MLTI_Interior_Att
 maxDistance=9900
 soundClassObject=&Fort_Vehicles
 volumeMultiplier=1
}
