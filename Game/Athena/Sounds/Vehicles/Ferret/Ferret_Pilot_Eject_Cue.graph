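# Sound cue for the pilot eject sound of the Ferret vehicle, in the vehicles sound class.
# A Random node picks between two equally weighted eject wave variants.
# It references a close-range biplane projectile attenuation asset and also carries natural-sound falloff override values.
# Its own concurrency override is enabled with a prevent-new resolution rule, so extra instances over the limit do not start.
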SoundCue /Game/Athena/Sounds/Vehicles/Ferret/Ferret_Pilot_Eject_Cue
root=Random_0
nodes [
 {id=Random_0, children [WavePlayer_0,WavePlayer_1], properties {weights [1,1]}}
 {id=WavePlayer_0, properties {soundWaveAssetPtr=&/Ferret_Pilot_Eject_01}}
 {id=WavePlayer_1, properties {soundWaveAssetPtr=&/Ferret_Pilot_Eject_02}}
]
properties {
 attenuationSettings=/Game/Sounds/Attenuation/Fort_Weapon_Projectile_Biplane_Close_Athena_Att
 bOverrideConcurrency=true
 maxDistance=15000
 soundClassObject=&Fort_Vehicles
 volumeMultiplier=1
 attenuationOverrides {distanceAlgorithm=NaturalSound, falloffDistance=4000, attenuationShapeExtents {x=1000}}
 concurrencyOverrides {resolutionRule=PreventNew}
}
