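# Sound cue for the low-fuel warning on the Ferret vehicle, routed through the vehicles sound class.
# A finite three-count loop wraps a short fixed delay into an equally weighted random pick of three warning waves, giving spaced repeats.
# It references the biplane close-movement attenuation asset and overrides concurrency to prevent new instances at the limit instead of stopping playing ones.
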SoundCue /Game/Athena/Sounds/Vehicles/Ferret/Ferret_Warning_FuelLow_Cue
root=Looping_0
nodes [
 {id=Delay_0, children [Random_0], properties {delayMax=0.2, delayMin=0.2}}
 {id=Looping_0, children [Delay_0], properties {bLoopIndefinitely=false, loopCount=3}}
 {id=Random_0, children [WavePlayer_0,WavePlayer_1,WavePlayer_3], properties {weights [1,1,1]}}
 {id=WavePlayer_0, properties {soundWaveAssetPtr=&/Ferret_Warning_FuelLow_01}}
 {id=WavePlayer_1, properties {soundWaveAssetPtr=&/Ferret_Warning_FuelLow_02}}
 {id=WavePlayer_3, properties {soundWaveAssetPtr=&/Ferret_Warning_FuelLow_03}}
]
properties {
 attenuationSettings=&/Fort_Biplane_Movement_Close_Att
 bOverrideConcurrency=true
 maxDistance=5300
 soundClassObject=&Fort_Vehicles
 volumeMultiplier=0.6
 attenuationOverrides {distanceAlgorithm=NaturalSound, falloffDistance=4000, attenuationShapeExtents {x=1000}}
 concurrencyOverrides {resolutionRule=PreventNew}
}
