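# Sound cue for the Ferret vehicle's out-of-fuel warning, in the vehicles sound class.
# A finite loop repeats a short fixed delay followed by an equally weighted random pick from three warning waves, producing a stuttered burst.
# Playback is slightly lowered in pitch and volume, references the Biplane close-movement attenuation, and overrides concurrency to block new instances once the limit is reached.
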SoundCue /Game/Athena/Sounds/Vehicles/Ferret/Ferret_Warning_FuelOut_Cue
root=Looping_0
nodes [
 {id=Delay_0, children [Random_1], properties {delayMax=0.15, delayMin=0.15}}
 {id=Looping_0, children [Delay_0], properties {bLoopIndefinitely=false, loopCount=6}}
 {id=Random_1, children [WavePlayer_2,WavePlayer_4,WavePlayer_5], properties {weights [1,1,1]}}
 {id=WavePlayer_2, properties {soundWaveAssetPtr=&/Ferret_Warning_FuelOut_01}}
 {id=WavePlayer_4, properties {soundWaveAssetPtr=&/Ferret_Warning_FuelOut_02}}
 {id=WavePlayer_5, properties {soundWaveAssetPtr=&/Ferret_Warning_FuelOut_03}}
]
properties {
 attenuationSettings=&/Fort_Biplane_Movement_Close_Att
 bOverrideConcurrency=true
 maxDistance=5300
 pitchMultiplier=0.9
 soundClassObject=&Fort_Vehicles
 volumeMultiplier=0.7
 attenuationOverrides {distanceAlgorithm=NaturalSound, falloffDistance=4000, attenuationShapeExtents {x=1000}}
 concurrencyOverrides {resolutionRule=PreventNew}
}
