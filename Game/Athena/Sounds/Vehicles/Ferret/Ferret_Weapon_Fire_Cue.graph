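# Sound cue for the Ferret vehicle's weapon fire, layering close and distant versions of the shot.
# A root mixer plays both branches, each randomly picking one of four waves (Close or Medium) with pitch and volume modulation.
# Each branch has its own Biplane projectile attenuation asset, close or distant, under a third-person ranged weapon fire sound class.
# The cue lowers overall volume and overrides concurrency to cap simultaneous shots.
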SoundCue /Game/Athena/Sounds/Vehicles/Ferret/Ferret_Weapon_Fire_Cue
root=Mixer_15
nodes [
 {
  id=Attenuation_80
  children [Modulator_0]
  properties {attenuationSettings=/Game/Sounds/Attenuation/Fort_Weapon_Projectile_Biplane_Close_Athena_Att}
 }
 {
  id=Attenuation_81
  children [Modulator_13]
  properties {attenuationSettings=/Game/Sounds/Attenuation/Fort_Weapon_Projectile_Biplane_Distant_Athena_Att}
 }
 {id=Mixer_0, children [Random_0], properties {inputVolume [1]}}
 {id=Mixer_15, children [SoundClass_55, SoundClass_56], properties {inputVolume [1,1]}}
 {id=Modulator_0, children [Mixer_0], properties {pitchMax=1.1, pitchMin=0.99, volumeMax=1, volumeMin=0.9}}
 {id=Modulator_13, children [Random_2], properties {volumeMin=1}}
 {id=Random_0, children [WavePlayer_0,WavePlayer_1,WavePlayer_2,WavePlayer_3], properties {weights [1,1,1,1]}}
 {id=Random_2, children [WavePlayer_8, WavePlayer_9, WavePlayer_10, WavePlayer_11], properties {weights [1,1,1,1]}}
 {
  id=SoundClass_55
  children [Attenuation_80]
  properties {soundClassOverride=&Fort_WeaponRanged_Fire_3P_PSM}
 }
 {
  id=SoundClass_56
  children [Attenuation_81]
  properties {soundClassOverride=&Fort_WeaponRanged_Fire_3P_PSM}
 }
 {id=WavePlayer_0, properties {soundWaveAssetPtr=&/Ferret_Weapon_Fire_Close_04}}
 {id=WavePlayer_1, properties {soundWaveAssetPtr=&/Ferret_Weapon_Fire_Close_01}}
 {id=WavePlayer_10, properties {soundWaveAssetPtr=&/Ferret_Weapon_Fire_Medium_03}}
 {id=WavePlayer_11, properties {soundWaveAssetPtr=&/Ferret_Weapon_Fire_Medium_04}}
 {id=WavePlayer_2, properties {soundWaveAssetPtr=&/Ferret_Weapon_Fire_Close_02}}
 {id=WavePlayer_3, properties {soundWaveAssetPtr=&/Ferret_Weapon_Fire_Close_03}}
 {id=WavePlayer_8, properties {soundWaveAssetPtr=&/Ferret_Weapon_Fire_Medium_01}}
 {id=WavePlayer_9, properties {soundWaveAssetPtr=&/Ferret_Weapon_Fire_Medium_02}}
]
properties {
 bOverrideConcurrency=true
 maxDistance=20000
 volumeMultiplier=0.3
 attenuationOverrides {distanceAlgorithm=NaturalSound, falloffDistance=4000, attenuationShapeExtents {x=1000}}
 concurrencyOverrides {maxCount=5}
}
