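# Sound cue for the Ferret vehicle's weapon overheat in Athena.
# A single wave player feeds the output with a reused mounted turret ballistic overheat sound, raised in pitch and lowered in volume.
# It uses the Vehicles sound class and a close biplane weapon projectile attenuation asset.
# A concurrency override is set to prevent new instances once the limit is reached.
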SoundCue /Game/Athena/Sounds/Vehicles/Ferret/Ferret_Weapon_Overheat_Cue
root=WavePlayer_2
nodes [{id=WavePlayer_2, properties {soundWaveAssetPtr=/Game/Sounds/Fort_Traps/MountedTurret/Turret_Ballistic_Overheat_02}}]
properties {
 attenuationSettings=/Game/Sounds/Attenuation/Fort_Weapon_Projectile_Biplane_Close_Athena_Att
 bOverrideConcurrency=true
 maxDistance=15000
 pitchMultiplier=1.2
 soundClassObject=&Fort_Vehicles
 volumeMultiplier=0.5
 attenuationOverrides {distanceAlgorithm=NaturalSound, falloffDistance=4000, attenuationShapeExtents {x=1000}}
 concurrencyOverrides {resolutionRule=PreventNew}
}
