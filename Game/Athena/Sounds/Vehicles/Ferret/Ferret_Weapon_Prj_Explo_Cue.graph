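# Sound cue for the Ferret vehicle weapon's projectile explosion, built from close impact sounds.
# A random node picks one of three impact waves with equal weight, then a modulator varies its volume and raises pitch by a random amount.
# It passes through a biplane close attenuation with overridden falloff and a third-person ranged weapon fire sound class into a mixer.
# Concurrency override caps simultaneous instances, stopping the oldest.
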SoundCue /Game/Athena/Sounds/Vehicles/Ferret/Ferret_Weapon_Prj_Explo_Cue
root=Mixer_7
nodes [
 {
  id=Attenuation_7
  children [Modulator_6]
  properties {attenuationSettings=&/Fort_Biplane_Movement_Close_Att, attenuationOverrides {falloffDistance=1600}}
 }
 {id=Mixer_7, children [SoundClass_4], properties {inputVolume [1]}}
 {id=Modulator_6, children [Random_0], properties {pitchMax=1.3, pitchMin=1, volumeMin=0.5}}
 {id=Random_0, children [WavePlayer_0,WavePlayer_1,WavePlayer_2], properties {weights [1,1,1]}}
 {
  id=SoundClass_4
  children [Attenuation_7]
  properties {soundClassOverride=&Fort_WeaponRanged_Fire_3P_PSM}
 }
 {id=WavePlayer_0, properties {soundWaveAssetPtr=&/Ferret_Weapon_Impact_Close_01}}
 {id=WavePlayer_1, properties {soundWaveAssetPtr=&/Ferret_Weapon_Impact_Close_03}}
 {id=WavePlayer_2, properties {soundWaveAssetPtr=&/Ferret_Weapon_Impact_Close_02}}
]
properties {bOverrideConcurrency=true, maxDistance=5300, volumeMultiplier=0.6, concurrencyOverrides {maxCount=2, resolutionRule=StopOldest}}
